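# Niagara system for the CopperMoggy emote that plays timed green screen effects on skeletal mesh bones, likely the gadget prop's screens.
# Delayed emitters layer ramping glitch and glow meshes, downward-drifting code meshes, then a brief sprite blink-out flare.
# Partition indices choose each particle's bone, size and shape, so one emitter covers several screens.
# System-age curves fade glows in, flicker them off and cut the code; Screen_Light has no renderers.
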NiagaraSystem /BRCosmetics/Animation/Game/MainPlayer/Emotes/CopperMoggy/FX/NS_EmoteCopperMoggy_ScreenFX
dependencies {
 usedDataInterfaces [&NiagaraDataInterfaceSkeletalMesh, &NiagaraDataInterfaceCurve, &NiagaraDataInterfaceArrayInt32]
 usedDynamicInputs [&Waveform]
 usedModules [
  &SystemState
  &self:SystemSpawnScript.SystemScriptSource.SystemScriptGraph.NiagaraNodeAssignment_0.SetVariables_#1
  &self:SystemSpawnScript.SystemScriptSource.SystemScriptGraph.NiagaraNodeAssignment_1.SetVariables_#2
 ]
}
emitters {
 BlinkOut {
  emitterUpdateScript [
   &EmitterState(Life Cycle Mode=::NewEnumerator1, Loop Behavior=::NewEnumerator1, Loop Delay=6, Loop Duration=1, UseLoopDelay=true)
   &SpawnBurst_Instantaneous(Spawn Count=1, Use Spawn Probability=false)
  ]
  particleSpawnScript [
   &InitializeParticle(
    Color=&MergeHelper_ColorSwitch(
     0=<rgba 0.1615 1 0.5001 0.5>
     1=<rgba 0.1615 1 0.5001 0.5>
     2=<rgba 0.0913 0.401 0.2075 0.5>
     3=<rgba 0.0913 0.401 0.2075 0.5>
     4=<rgba 0.1615 1 0.5001 0.5>
     5=<rgba 0.1615 1 0.5001 0.5>
     6=<rgba 0.0913 0.401 0.2075 0.5>
     Index=$Particles.PartitionParticles.Partition
     Num=::NewEnumerator5
    )
    Lifetime=0.23
    Sprite Size Mode=::NewEnumerator3
    Uniform Sprite Size=&Multiply_Float(
     A=100
     B=&MergeHelper_FloatSwitch(0=1, 1=0.8, 2=0.7, 3=0.6, 4=0.8, 5=0.6, 6=0.6, Index=$Particles.PartitionParticles.Partition, Num=::NewEnumerator5)
    )
    UsePositionOffset=true
    Write Material Random=true
   )
   &ShapeLocation(MaskSpawnGroup=false, Override Seed=false, Sphere Radius=0, UseOwnerScale=true)
   &CameraOffset(Camera Offset Amount=20)
   /SetVariables(Particles.SizeWideFlare:Vector2f=&Multiply_Vector2D_ByFloat(
     Float=&MergeHelper_FloatSwitch(0=1, 1=0.8, 2=0.7, 3=0.6, 4=1.2, 5=0.6, 6=0.6, Index=$Particles.PartitionParticles.Partition, Num=::NewEnumerator5)
     Vector2D=<230 3>
    )) {isSetParameters=true, module=SetVariables_#3}
   &InitialMeshOrientation(Mesh Orientation Mode=::NewEnumerator2, Orientation Vector=<0 1 0>, Rotation=<0.25 0 0>, Use Rotation Vector=true)
   &ApplyInitialForces()
   &AlignSpriteToMeshOrientation(SetAlignment=true, SetFacing=true)
   &SkeletalMeshLocation(
    Bone Sampling Mode=::NewEnumerator4
    MaskSpawnGroup=false
    Sampled Position Offset=<0 0 -3>
    Skeletal Mesh=DataInterface() {propertyValues {
      DefaultMesh=null
      ExcludeBoneName=root
      OptionalComponentClass=null
      PreviewMesh=/BRCosmetics/Animation/Props/CopperMoggy/CopperMoggy_Gadget
      SkinningMode=SkinOnTheFly
      SoftSourceActor=null
      SourceMode=Default
      UvSetIndex=0
      WholeMeshLOD=-1
      bExcludeBone=false
      bReadDeformedGeometry=true
      bRequireCurrentFrameData=true
      ComponentTags []
      FilteredBones [offset]
      FilteredSockets []
      MeshUserParameter {parameter {name=None, typeDefHandle {registeredTypeIndex=1}, varData []}}
      SamplingRegions []
     }}
   )
  ]
  particleUpdateScript [
   &ParticleState()
   &SolveForcesAndVelocity(Clamp Velocity=false, Limit Acceleration=false)
   &ScaleColor(
    Scale Alpha=&FloatFromCurve(FloatCurve=DataInterface() {propertyValues {Curve {keys [
         <at 0.1 0 RCIM_Cubic>
         {interpMode=RCIM_Cubic, leaveTangent=-2.0997, tangentMode=RCTM_Break, time=0.2521, value=1}
         {arriveTangentWeight=0.7369, interpMode=RCIM_Cubic, tangentMode=RCTM_Break, time=1, value=1}
        ]}}})
    Scale RGB=&VectorFromFloat(Value=&FloatFromCurve(
      FloatCurve=DataInterface() {propertyValues {Curve {keys [
          <at 0.1 0 RCIM_Cubic>
          {interpMode=RCIM_Cubic, leaveTangent=-2.0997, tangentMode=RCTM_Break, time=0.2521, value=1}
          {arriveTangentWeight=0.7369, interpMode=RCIM_Cubic, tangentMode=RCTM_Break, time=0.6, value=1}
         ]}}}
      Scale Curve=5
     ))
    ScaleA=true
    ScaleRGB=true
   )
   &ScaleSpriteSize(Uniform Curve Sprite Scale=DataInterface() {propertyValues {Curve {keys [
        <at 0.3 0 RCIM_Cubic>
        {arriveTangentWeight=0.1, interpMode=RCIM_Cubic, leaveTangentWeight=0.0333, tangentMode=RCTM_User, time=0.6, value=1}
        {arriveTangentWeight=0.0898, interpMode=RCIM_Cubic, leaveTangentWeight=0.2693, tangentMode=RCTM_User, time=0.7, value=1}
        {interpMode=RCIM_Cubic, tangentMode=RCTM_User, time=1, value=0}
       ]}}})
   &FN_Sprite_Attributes(
    Color=&MakeLinearColorFromVectorAndFloat(
     "Vector (RGB)"=&MakeVectorFromLinearColorRGB(Color=<rgba 6.4901 30 11.2936 3>)
     "Float (Alpha)"=&Multiply_Float(
      A=1
      B=&FloatFromCurve(FloatCurve=DataInterface() {propertyValues {Curve {keys [{interpMode=RCIM_Cubic, tangentMode=RCTM_User, time=0, value=1}, {interpMode=RCIM_Cubic, tangentMode=RCTM_User, time=1, value=0.5}]}}})
     )
    )
    Create Color Attribute=true
    Module Output:ENiagara_AttributeSamplingApplyOutput=::NewEnumerator1
   )
   /SetVariables(Particles.SizeWideFlare:Vector2f=&Multiply_Vector2D(
     A=$Particles.Initial.SizeWideFlare
     B=&MakeVector2D(
      X=&FloatFromCurve(FloatCurve=DataInterface() {propertyValues {Curve {keys [<at 0.1 0>,<at 0.1442 0.4>,<at 0.2442 1>,<at 0.3716 1>,<at 1 0>]}}})
      Y=&FloatFromCurve(FloatCurve=DataInterface() {propertyValues {Curve {keys [<at 0.2 1>,<at 0.5 0.3>]}}})
     )
    )) {isSetParameters=true, module=SetVariables_#4}
  ]
  rendererClasses [&NiagaraSpriteRendererProperties]
  renderers [
   {rendererClass=&NiagaraSpriteRendererProperties}
   {rendererClass=&NiagaraSpriteRendererProperties, rendererIndex=1}
  ]
 }
 CodeFalling {
  emitterSpawnScript [/SetVariables(
    Emitter.Offset:LinearColor=&RandomRangeLinearColor(Minimum=<rgba 0 0 0 0>, Maximum=<rgba 0.2 0 0 0>)
    Emitter.UV Scale:LinearColor=&RandomRangeLinearColor(Minimum=<rgba 2 1 0 0>, Maximum=<rgba 2 0.5 0 0>)
   ) {isSetParameters=true}]
  emitterUpdateScript [
   &EmitterState(Life Cycle Mode=::NewEnumerator1, Loop Behavior=::NewEnumerator1, Loop Delay=3.71, Loop Duration=2, UseLoopDelay=true)
   &SpawnRate(SpawnRate=40, Use Spawn Probability=false)
  ]
  particleSpawnScript [
   &PartitionParticles(Number Of Partitions=4)
   &InitializeParticle(
    Color=<rgba 0.1615 1 0.5001 0.25>
    Lifetime Max=1
    Lifetime Min=0.645
    Lifetime Mode=::NewEnumerator1
    Mesh Scale Max=<0.15 1 0.23>
    Mesh Scale Min=<0.1 0.2 0.15>
    Mesh Scale Mode=::NewEnumerator2
    Position Offset=&MergeHelper_VectorSwitch(
     0=<35 65 156>
     1=<-27 65 151>
     2=<-60 65 143>
     3=<65 65 135>
     4=<-20 65 120>
     5=<3 65 176>
     6=<63 65 168>
     Index=$Particles.PartitionParticles.Partition
     Num=::NewEnumerator5
    )
    Sprite Size Mode=::NewEnumerator3
    Uniform Sprite Size=&Multiply_Float(
     A=50
     B=&MergeHelper_FloatSwitch(0=1, 1=0.8, 2=0.7, 3=0.6, 4=1, 5=0.6, 6=0.6, Index=$Particles.PartitionParticles.Partition, Num=::NewEnumerator5)
    )
    UsePositionOffset=false
    Write Material Random=true
   )
   &CameraOffset(Camera Offset Amount=-10)
   /SetVariables(Particles.SizeWideFlare:Vector2f=&Multiply_Vector2D_ByFloat(
     Float=&MergeHelper_FloatSwitch(0=1, 1=0.8, 2=0.7, 3=0.6, 4=1, 5=0.6, 6=0.6, Index=$Particles.PartitionParticles.Partition, Num=::NewEnumerator5)
     Vector2D=<225 40>
    )) {isSetParameters=true, module=SetVariables_#3}
   &InitialMeshOrientation(Mesh Orientation Mode=::NewEnumerator2, Orientation Vector=<0 1 0>, Rotation=<0.25 0 0>, Use Rotation Vector=true)
   &SkeletalMeshLocation(
    Bone / Socket Index=&SelectIntFromArray(Direct Array Index=$Particles.PartitionParticles.Partition, Int Selection Array=DataInterface() {propertyValues {IntData [0,2,3,4]}})
    Bone Sampling Mode=::NewEnumerator4
    MaskSpawnGroup=false
    Position Sampling=::NewEnumerator2
    Sampled Position Offset=&MergeHelper_VectorSwitch(0=<0 0 -15>, 1=<0 0 0>, 2=<0 0 0>, 3=<0 0 0>, Index=$Particles.PartitionParticles.Partition, Num=::NewEnumerator2)
    Skeletal Mesh=$System.Skeletal Mesh Reader
   )
   &ShapeLocation(
    Box Size=&MergeHelper_VectorSwitch(0=<25 0 0.001>, 1=<20 0 0.001>, 2=<15 0 0.001>, 3=<40 0 0.001>, Index=$Particles.PartitionParticles.Partition, Num=::NewEnumerator2)
    MaskSpawnGroup=false
    Override Seed=false
    Shape Origin=$Particles.SkeletalMeshLocation.SampledPosition
    Shape Primitive=::NewEnumerator2
    Surface Only=false
    UseOwnerScale=true
   )
   &AddVelocity(Velocity=&RandomRangeVector(Minimum=<0 0 -25>, Maximum=<0 0 -50>))
   /SetVariables(Particles.VisibilityTag=&SpawnGroup()) {isSetParameters=true}
   &DynamicMaterialParameters(
    Index 0 Param 1=&RandomRangeFloat(Minimum=0, Maximum=5)
    Module.Index0Float0Write=true
    Module.Index0Float1Write=true
    Module.Index0Float2Write=true
    Module.Index0Float3Write=true
   )
  ]
  particleUpdateScript [
   &ParticleState()
   &SolveForcesAndVelocity(Clamp Velocity=false, Limit Acceleration=false)
   &ScaleColor(
    Scale Alpha=&FloatFromCurve(FloatCurve=DataInterface() {propertyValues {Curve {keys [
         {interpMode=RCIM_Cubic, leaveTangentWeight=0.6455, tangentMode=RCTM_User, time=0.2, value=0}
         {arriveTangentWeight=0.05, interpMode=RCIM_Cubic, leaveTangentWeight=0.5814, tangentMode=RCTM_Break, time=0.5, value=1}
         {arriveTangent=-2.526, arriveTangentWeight=0.7369, interpMode=RCIM_Cubic, tangentMode=RCTM_Break, time=1, value=0}
        ]}}})
    Scale RGB=&VectorFromFloat(Value=&FloatFromCurve(
      FloatCurve=DataInterface() {propertyValues {Curve {keys [
          <at 0.1 0 RCIM_Cubic>
          {interpMode=RCIM_Cubic, leaveTangent=-2.0997, tangentMode=RCTM_Break, time=0.2521, value=1}
          {arriveTangentWeight=0.7369, interpMode=RCIM_Cubic, tangentMode=RCTM_Break, time=0.6, value=1}
         ]}}}
      Scale Curve=5
     ))
    ScaleA=true
    ScaleRGB=true
   )
   &ScaleColor(
    Scale Alpha=&FloatFromCurve(FloatCurve=DataInterface() {propertyValues {Curve {keys [<at 5.75 1>,<at 5.965 0>]}}}, CurveIndex=$System.Age)
    ScaleA=true
    ScaleRGB=true
   ) {module=ScaleColor001}
   &ScaleSpriteSize(Uniform Curve Sprite Scale=DataInterface() {propertyValues {Curve {keys [
        <at 0.3 0 RCIM_Cubic>
        {interpMode=RCIM_Cubic, tangentMode=RCTM_User, time=0.4, value=1}
        {interpMode=RCIM_Cubic, tangentMode=RCTM_User, time=1, value=0}
       ]}}})
   &FN_Sprite_Attributes(
    Color=&MakeLinearColorFromVectorAndFloat(
     "Vector (RGB)"=&MakeVectorFromLinearColorRGB(Color=<rgba 0.649 3 1.1294 1>)
     "Float (Alpha)"=&Multiply_Float(
      A=1
      B=&FloatFromCurve(FloatCurve=DataInterface() {propertyValues {Curve {keys [{interpMode=RCIM_Cubic, tangentMode=RCTM_User, time=0, value=1}, {interpMode=RCIM_Cubic, tangentMode=RCTM_User, time=1, value=0.5}]}}})
     )
    )
    Create Color Attribute=true
    Module Output:ENiagara_AttributeSamplingApplyOutput=::NewEnumerator1
   )
   /SetVariables(Particles.SizeWideFlare:Vector2f=&Multiply_Vector2D(
     A=$Particles.Initial.SizeWideFlare
     B=&MakeVector2D(
      X=&FloatFromCurve(FloatCurve=DataInterface() {propertyValues {Curve {keys [<at 0.1 0>,<at 0.2 0.4>,<at 0.3 1>,<at 1 0>]}}})
      Y=&FloatFromCurve(FloatCurve=DataInterface() {propertyValues {Curve {keys [<at 0.2 1>,<at 0.5 0.1>]}}})
     )
    )) {isSetParameters=true, module=SetVariables_#4}
  ]
  rendererClasses [&NiagaraMeshRendererProperties]
  renderers [{rendererClass=&NiagaraMeshRendererProperties, rendererIndex=0}]
 }
 Screen_Glitching {
  emitterUpdateScript [
   &EmitterState(Life Cycle Mode=::NewEnumerator1, Loop Behavior=::NewEnumerator1, Loop Delay=3.01, Loop Duration=2.8, UseLoopDelay=true)
   &SpawnRate(
    Spawn Probability=0.3
    SpawnRate=&FloatFromCurve(
     FloatCurve=DataInterface() {propertyValues {Curve {keys [<at 0 0 RCIM_Cubic>,<at 0.5 1 RCIM_Cubic>]}}}
     CurveIndex=$Emitter.NormalizedLoopAge
     Scale Curve=80
    )
    Use Spawn Probability=true
   )
  ]
  particleSpawnScript [
   &PartitionParticles(Number Of Partitions=7)
   &InitializeParticle(
    Color=<rgba 0.1615 1 0.5001 0.5>
    Lifetime Max=0.17
    Lifetime Min=0.05
    Lifetime Mode=::NewEnumerator1
    Mesh Scale Max=<0.2 1 1>
    Mesh Scale Min=<0.1 0.1 0.1>
    Mesh Scale Mode=::NewEnumerator2
    Position Offset=&MergeHelper_VectorSwitch(
     0=<35 65 156>
     1=<-27 65 151>
     2=<-60 65 143>
     3=<65 65 135>
     4=<-20 65 120>
     5=<3 65 176>
     6=<63 65 168>
     Index=$Particles.PartitionParticles.Partition
     Num=::NewEnumerator5
    )
    UsePositionOffset=false
    Write Material Random=true
   )
   &CameraOffset(Camera Offset Amount=3)
   /SetVariables(Particles.SizeWideFlare:Vector2f=&Multiply_Vector2D_ByFloat(
     Float=&MergeHelper_FloatSwitch(0=1, 1=0.8, 2=0.7, 3=0.6, 4=1, 5=0.6, 6=0.6, Index=$Particles.PartitionParticles.Partition, Num=::NewEnumerator5)
     Vector2D=<225 40>
    )) {isSetParameters=true, module=SetVariables_#3}
   &InitialMeshOrientation(Mesh Orientation Mode=::NewEnumerator2, Orientation Vector=<0 1 0>, Rotation=<0.25 0 0>, Use Rotation Vector=true)
   &SkeletalMeshLocation(
    Bone / Socket Index=$Particles.PartitionParticles.Partition
    Bone Sampling Mode=::NewEnumerator4
    MaskSpawnGroup=false
    Skeletal Mesh=$System.Skeletal Mesh Reader
   )
   &ShapeLocation(
    Box Size=&MergeHelper_VectorSwitch(
     0=<15 0 30>
     1=<15 0 20>
     2=<10 0 10>
     3=<5 0 5>
     4=<35 0 5>
     5=<5 0 7.5>
     6=<5 0 5>
     Index=$Particles.PartitionParticles.Partition
     Num=::NewEnumerator5
    )
    MaskSpawnGroup=false
    Override Seed=false
    Shape Origin=$Particles.SkeletalMeshLocation.SampledPosition
    Shape Primitive=::NewEnumerator2
    Surface Only=false
    UseOwnerScale=true
   )
  ]
  particleUpdateScript [
   &ParticleState()
   &ScaleMeshSize(Scale Factor=&VectorFromFloat(Value=&FloatFromCurve(FloatCurve=DataInterface() {propertyValues {Curve {keys [<at 0 1>,<at 0.2 1.1>,<at 1 1>]}}})))
   &SolveForcesAndVelocity(Clamp Velocity=false, Limit Acceleration=false)
   &ScaleColor(
    Scale Alpha=&FloatFromCurve(
     FloatCurve=DataInterface() {propertyValues {Curve {keys [
         {arriveTangent=6.3774, interpMode=RCIM_Cubic, leaveTangent=6.3774, tangentMode=RCTM_User, time=0.1, value=0}
         {arriveTangentWeight=0.05, interpMode=RCIM_Cubic, leaveTangentWeight=0.5814, tangentMode=RCTM_Break, time=0.25, value=1}
         {arriveTangent=-2.526, arriveTangentWeight=0.7369, interpMode=RCIM_Cubic, tangentMode=RCTM_Break, time=1, value=0}
        ]}}}
     Scale Curve=&RandomRangeFloat(Minimum=0.3, Maximum=0.6)
    )
    Scale RGB=&VectorFromFloat(Value=&FloatFromCurve(
      FloatCurve=DataInterface() {propertyValues {Curve {keys [
          <at 0.1 0 RCIM_Cubic>
          {interpMode=RCIM_Cubic, leaveTangent=-2.0997, tangentMode=RCTM_Break, time=0.2521, value=1}
          {arriveTangentWeight=0.7369, interpMode=RCIM_Cubic, tangentMode=RCTM_Break, time=0.6, value=1}
         ]}}}
      Scale Curve=&RandomRangeFloat(Minimum=1, Maximum=5)
     ))
    ScaleA=true
    ScaleRGB=true
   )
   &ScaleSpriteSize(Uniform Curve Sprite Scale=DataInterface() {propertyValues {Curve {keys [
        <at 0.3 0 RCIM_Cubic>
        {interpMode=RCIM_Cubic, tangentMode=RCTM_User, time=0.4, value=1}
        {interpMode=RCIM_Cubic, tangentMode=RCTM_User, time=1, value=0}
       ]}}})
   &FN_Sprite_Attributes(
    Color=&MakeLinearColorFromVectorAndFloat(
     "Vector (RGB)"=&MakeVectorFromLinearColorRGB(Color=<rgba 0.649 3 1.1294 1>)
     "Float (Alpha)"=&Multiply_Float(
      A=1
      B=&FloatFromCurve(FloatCurve=DataInterface() {propertyValues {Curve {keys [{interpMode=RCIM_Cubic, tangentMode=RCTM_User, time=0, value=1}, {interpMode=RCIM_Cubic, tangentMode=RCTM_User, time=1, value=0.5}]}}})
     )
    )
    Create Color Attribute=true
    Module Output:ENiagara_AttributeSamplingApplyOutput=::NewEnumerator1
   )
   /SetVariables(Particles.SizeWideFlare:Vector2f=&Multiply_Vector2D(
     A=$Particles.Initial.SizeWideFlare
     B=&MakeVector2D(
      X=&FloatFromCurve(FloatCurve=DataInterface() {propertyValues {Curve {keys [<at 0.1 0>,<at 0.2 0.4>,<at 0.3 1>,<at 1 0>]}}})
      Y=&FloatFromCurve(FloatCurve=DataInterface() {propertyValues {Curve {keys [<at 0.2 1>,<at 0.5 0.1>]}}})
     )
    )) {isSetParameters=true, module=SetVariables_#4}
   &DynamicMaterialParameters(
    Index 0 Param 1=&RandomRangeFloat(Minimum=1, Maximum=5)
    Module.Index0Float0Write=true
    Module.Index0Float1Write=true
    Module.Index0Float2Write=true
    Module.Index0Float3Write=true
   )
  ]
  rendererClasses [&NiagaraMeshRendererProperties]
  renderers [{rendererClass=&NiagaraMeshRendererProperties, rendererIndex=0}]
 }
 Screen_Glows {
  emitterUpdateScript [
   &EmitterState(Life Cycle Mode=::NewEnumerator1, Loop Behavior=::NewEnumerator1, Loop Delay=3.01, Loop Duration=2.8, UseLoopDelay=true)
   &SpawnRate(
    Spawn Probability=0.5
    SpawnRate=&FloatFromCurve(
     FloatCurve=DataInterface() {propertyValues {Curve {keys [<at 0 0 RCIM_Cubic>, {arriveTangent=3.3732, interpMode=RCIM_Cubic, leaveTangent=3.3732, tangentMode=RCTM_User, time=0.5, value=1}]}}}
     CurveIndex=$Emitter.NormalizedLoopAge
     Scale Curve=75
    )
    Use Spawn Probability=true
   )
  ]
  particleSpawnScript [
   &PartitionParticles(Number Of Partitions=7)
   &InitializeParticle(
    Color Maximum=<rgba 0.1615 1 0.5001 0.35>
    Color Minimum=<rgba 0.1615 1 0.5001 0.1>
    Color Mode=::NewEnumerator2
    Lifetime Max=0.35
    Lifetime Min=0.1
    Lifetime Mode=::NewEnumerator1
    Mesh Scale=&MergeHelper_VectorSwitch(
     0=<0.09 1.4 2.546>
     1=<0.095 1.1 1.2988>
     2=<0.078 0.78 1.25>
     3=<0.07 0.6 0.98>
     4=<0.125 0.55 2.3>
     5=<0.058 0.58 1.05>
     6=<0.06 0.42 0.85>
     Index=$Particles.PartitionParticles.Partition
     Num=::NewEnumerator5
    )
    Mesh Scale Mode=::NewEnumerator4
    Position Offset=&MergeHelper_VectorSwitch(
     0=<35 65 156>
     1=<-27 65 151>
     2=<-60 65 143>
     3=<65 65 135>
     4=<-20 65 120>
     5=<3 65 176>
     6=<63 65 168>
     Index=$Particles.PartitionParticles.Partition
     Num=::NewEnumerator5
    )
    UsePositionOffset=false
    Write Material Random=true
   )
   &CameraOffset(Camera Offset Amount=3)
   /SetVariables(Particles.SizeWideFlare:Vector2f=&Multiply_Vector2D_ByFloat(
     Float=&MergeHelper_FloatSwitch(0=1, 1=0.8, 2=0.7, 3=0.6, 4=1, 5=0.6, 6=0.6, Index=$Particles.PartitionParticles.Partition, Num=::NewEnumerator5)
     Vector2D=<225 40>
    )) {isSetParameters=true, module=SetVariables_#3}
   &InitialMeshOrientation(
    Mesh Orientation Mode=::NewEnumerator2
    Orientation Vector=<0 0 1>
    Rotation=&MergeHelper_VectorSwitch(
     0=<0.1996 0 0>
     1=<0.2936 0 0>
     2=<0.32 0 0>
     3=<0.1828 0 0>
     4=<0.2684 0 0>
     5=<0.25 0 0>
     6=<0.1828 0 0>
     Index=$Particles.PartitionParticles.Partition
     Num=::NewEnumerator5
    )
    Use Rotation Vector=true
   )
   &SkeletalMeshLocation(
    Bone / Socket Index=&Multiply_Int(
     A=$Particles.PartitionParticles.Partition
     B=&MakeIntFromFloat(Float=&FloatFromCurve(
       FloatCurve=DataInterface() {propertyValues {Curve {keys [
           {arriveTangent=-0.0576, arriveTangentWeight=2.1277, interpMode=RCIM_Cubic, leaveTangentWeight=2.1277, time=1.055, value=0}
           {
            arriveTangent=1.0761
            arriveTangentWeight=0.6688
            interpMode=RCIM_Cubic
            leaveTangent=1.0761
            leaveTangentWeight=0.6878
            tangentMode=RCTM_User
            time=3.5
            value=1
           }
          ]}}}
       CurveIndex=$System.Age
      ))
    )
    Bone Sampling Mode=::NewEnumerator4
    MaskSpawnGroup=false
    Skeletal Mesh=$System.Skeletal Mesh Reader
   )
   /SetVariables(Particles.VisibilityTag=&SpawnGroup()) {isSetParameters=true}
  ]
  particleUpdateScript [
   &ParticleState()
   &SolveForcesAndVelocity(Clamp Velocity=false, Limit Acceleration=false)
   &ScaleColor(
    Scale Alpha=&FloatFromCurve(
     FloatCurve=DataInterface() {propertyValues {Curve {keys [
         {arriveTangent=6.3774, interpMode=RCIM_Cubic, leaveTangent=6.3774, tangentMode=RCTM_User, time=0.1, value=0}
         {arriveTangentWeight=0.05, interpMode=RCIM_Cubic, leaveTangentWeight=0.5814, tangentMode=RCTM_Break, time=0.25, value=1}
         {arriveTangent=-2.526, arriveTangentWeight=0.7369, interpMode=RCIM_Cubic, tangentMode=RCTM_Break, time=1, value=0}
        ]}}}
     Scale Curve=&RandomRangeFloat(Minimum=0.1, Maximum=0.6)
    )
    Scale RGB=&VectorFromFloat(Value=&FloatFromCurve(
      FloatCurve=DataInterface() {propertyValues {Curve {keys [
          <at 0.1 0 RCIM_Cubic>
          {interpMode=RCIM_Cubic, leaveTangent=-2.0997, tangentMode=RCTM_Break, time=0.2521, value=1}
          {arriveTangentWeight=0.7369, interpMode=RCIM_Cubic, tangentMode=RCTM_Break, time=0.6, value=1}
         ]}}}
      Scale Curve=0.8
     ))
    ScaleA=true
    ScaleRGB=true
   )
   &ScaleColor(
    Scale Alpha=&FloatFromCurve(
     FloatCurve=DataInterface() {propertyValues {Curve {keys [
         <at 1.055 0 RCIM_Cubic>
         {arriveTangentWeight=0.6688, interpMode=RCIM_Cubic, leaveTangentWeight=0.6878, tangentMode=RCTM_User, time=3, value=1}
         {arriveTangentWeight=0.6667, interpMode=RCIM_Cubic, leaveTangentWeight=0.0635, tangentMode=RCTM_Break, time=5, value=1}
         {arriveTangentWeight=0.0231, leaveTangentWeight=0.0267, time=5.453, value=1}
         {arriveTangentWeight=0.0665, leaveTangentWeight=0.0791, time=5.5445, value=0}
         {arriveTangentWeight=0.0317, leaveTangentWeight=0.0209, time=5.6367, value=1}
         {arriveTangentWeight=0.0323, leaveTangentWeight=0.0334, time=5.7276, value=0}
         {arriveTangentWeight=0.0216, leaveTangentWeight=0.0358, time=5.8183, value=1}
         {arriveTangentWeight=0.0358, time=5.9, value=0}
        ]}}}
     CurveIndex=$System.Age
    )
    ScaleA=true
    ScaleRGB=true
   ) {module=ScaleColor001}
   &ScaleSpriteSize(Uniform Curve Sprite Scale=DataInterface() {propertyValues {Curve {keys [
        <at 0.3 0 RCIM_Cubic>
        {interpMode=RCIM_Cubic, tangentMode=RCTM_User, time=0.4, value=1}
        {interpMode=RCIM_Cubic, tangentMode=RCTM_User, time=1, value=0}
       ]}}})
   &FN_Sprite_Attributes(
    Color=&MakeLinearColorFromVectorAndFloat(
     "Vector (RGB)"=&MakeVectorFromLinearColorRGB(Color=<rgba 0.649 3 1.1294 1>)
     "Float (Alpha)"=&Multiply_Float(
      A=1
      B=&FloatFromCurve(FloatCurve=DataInterface() {propertyValues {Curve {keys [{interpMode=RCIM_Cubic, tangentMode=RCTM_User, time=0, value=1}, {interpMode=RCIM_Cubic, tangentMode=RCTM_User, time=1, value=0.5}]}}})
     )
    )
    Create Color Attribute=true
    Module Output:ENiagara_AttributeSamplingApplyOutput=::NewEnumerator1
   )
   /SetVariables(Particles.SizeWideFlare:Vector2f=&Multiply_Vector2D(
     A=$Particles.Initial.SizeWideFlare
     B=&MakeVector2D(
      X=&FloatFromCurve(FloatCurve=DataInterface() {propertyValues {Curve {keys [<at 0.1 0>,<at 0.2 0.4>,<at 0.3 1>,<at 1 0>]}}})
      Y=&FloatFromCurve(FloatCurve=DataInterface() {propertyValues {Curve {keys [<at 0.2 1>,<at 0.5 0.1>]}}})
     )
    )) {isSetParameters=true, module=SetVariables_#4}
   &DynamicMaterialParameters(
    Index 0 Param 1=0.35
    Index 0 Param 2=0.35
    Module.Index0Float0Write=true
    Module.Index0Float1Write=true
    Module.Index0Float2Write=true
    Module.Index0Float3Write=true
   )
  ]
  rendererClasses [&NiagaraMeshRendererProperties]
  renderers [
   {rendererClass=&NiagaraMeshRendererProperties, rendererIndex=0}
   {rendererClass=&NiagaraMeshRendererProperties, rendererIndex=1}
  ]
 }
 Screen_Light {rendererClasses [], renderers []}
}
summary {
 systemName=NS_EmoteCopperMoggy_ScreenFX
 emitters [
  {emitterName=BlinkOut, rendererClasses [&NiagaraSpriteRendererProperties]}
  {emitterName=CodeFalling, rendererClasses [&NiagaraMeshRendererProperties]}
  {emitterName=Screen_Glitching, rendererClasses [&NiagaraMeshRendererProperties]}
  {emitterName=Screen_Light, rendererClasses []}
  {emitterName=Screen_Glows, rendererClasses [&NiagaraMeshRendererProperties]}
 ]
 userVariables []
}
