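# Niagara effect that plays a green glitch burst for prop spawning in the CopperMoggy emote.
# Two near-identical emitters, offset to opposite sides along X, fire short, staggered bursts split by spawn group into a flickering sprite flare and long and square mesh glitches.
# Curves and random values animate size and dynamic material parameters.
# Colors are tinted by a time-of-day curve, and glitch size is scaled by LOD distance.
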NiagaraSystem /BRCosmetics/Animation/Game/MainPlayer/Emotes/CopperMoggy/FX/NS_Emote_CopperMoggy_PropSpawn
dependencies {
 usedDataInterfaces [&NiagaraDataInterfaceColorCurve, &NiagaraDataInterfaceCurve]
 usedDynamicInputs []
 usedModules [&SystemState]
}
emitters {
 PropSpawn_Glitch {
  emitterSpawnScript [/SetVariables(
    Emitter.ColorA:LinearColor=<rgba 0 0.3281 0.0898 1>
    Emitter.ColorB:LinearColor=<rgba 0 0.3177 0.0115 1>
    Emitter.MasterSizeMultiplier=0.7
    Emitter.ShapeLocationScale:Vector3f=<1.1 1.5 0.65>
   ) {isSetParameters=true, module=SetVariables_#1}]
  emitterUpdateScript [
   &EmitterState(Life Cycle Mode=::NewEnumerator1, Loop Behavior=::NewEnumerator1, Loop Duration=1, UseLoopDelay=false)
   &SpawnBurst_Instantaneous(Use Spawn Probability=false, UseLoopCountLimit=false)
   &SpawnBurst_Instantaneous(Spawn Count=10, Spawn Group=2, Use Spawn Probability=false, UseLoopCountLimit=false) {module=SpawnBurst_Instantaneous008}
   &SpawnBurst_Instantaneous(Spawn Count=5, Spawn Group=2, Spawn Time=0.15, Use Spawn Probability=false, UseLoopCountLimit=false) {module=SpawnBurst_Instantaneous012}
   &SpawnBurst_Instantaneous(Spawn Count=3, Spawn Group=3, Use Spawn Probability=false, UseLoopCountLimit=false) {module=SpawnBurst_Instantaneous011}
   &SpawnBurst_Instantaneous(Spawn Count=3, Spawn Group=3, Spawn Time=0.05, Use Spawn Probability=false, UseLoopCountLimit=false) {module=SpawnBurst_Instantaneous010}
   &SpawnBurst_Instantaneous(Spawn Count=5, Spawn Group=3, Spawn Time=0.1, Use Spawn Probability=false, UseLoopCountLimit=false) {module=SpawnBurst_Instantaneous009}
  ]
  particleSpawnScript [
   /SetVariables(Particles.VisibilityTag=&SpawnGroup()) {isSetParameters=true, module=SetVariables_#2}
   /SetVariables(Particles.Random Range:NiagaraFloat=&RandomRangeFloat(Minimum=0.2, Maximum=0.8)) {isSetParameters=true, module=SetVariables_#3}
   &InitializeParticle(
    Color Mode=::NewEnumerator0
    Lifetime=&MergeHelper_FloatSwitch(
     0=0.27
     1=0.3
     2=&RandomRangeFloat(Minimum=0.25, Maximum=0.3)
     3=&RandomRangeFloat(Minimum=0.25, Maximum=0.3)
     4=0.2
     6=&RandomRangeFloat(Minimum=0.55, Maximum=0.55)
     7=&RandomRangeFloat(Minimum=0.5, Maximum=0.62)
     8=&RandomRangeFloat(Minimum=0.25, Maximum=0.5)
     9=&RandomRangeFloat(Minimum=0.25, Maximum=0.5)
     Index=&SpawnGroup()
     Num=::NewEnumerator3
    )
    Position Offset=<-15 0 0>
    UsePositionOffset=true
    Write Material Random=true
   ) {module=InitializeParticle001}
   &CameraOffset(Camera Offset Amount=&MergeHelper_FloatSwitch(0=2, 1=2, 2=10, 3=10, 4=2, Index=0, Num=::NewEnumerator3))
   /SetVariables(
    Particles.Bright Flare Size:Vector2f=&Vector2DFromFloat(Value=130)
    Particles.Bright Flare color:LinearColor=$Emitter.ColorB
   ) {isSetParameters=true, module=SetVariables_#4}
   /SetVariables(
    Particles.Long Glitch Dynam 1:Vector4f=&MakeVector4(X=0, Y=0, Z=0, W=&RandomRangeFloat(Minimum=-1, Maximum=0.1))
    Particles.Long Glitch Dynam 2:Vector4f=&MakeVector4(X=&RandomRangeFloat(Minimum=-100, Maximum=100), Y=0, Z=0, W=0)
    Particles.Long Glitch Size:Vector3f=&RandomRangeVector(Minimum=<1 3 0.5>, Maximum=<2 5 1>)
   ) {isSetParameters=true, module=SetVariables_#5}
   /SetVariables(
    Particles.Square Glitch Dynam 1:Vector4f=&MakeVector4(X=0, Y=0, Z=0, W=&RandomRangeFloat(Minimum=-1, Maximum=0.1))
    Particles.Square Glitch Dynam 2:Vector4f=&MakeVector4(X=&RandomRangeFloat(Minimum=-100, Maximum=100), Y=0, Z=0, W=0)
    Particles.Square Glitch Size:Vector3f=&RandomRangeVector(Minimum=<1 1 1>, Maximum=<1.5 1.5 1.5>)
   ) {isSetParameters=true, module=SetVariables_#6}
   &ShapeLocation(
    MaskSpawnGroup=false
    Non Uniform Scale=$Emitter.ShapeLocationScale
    Override Seed=false
    Sphere Radius=&Multiply_Float(
     A=&MergeHelper_FloatSwitch(0=0, 1=0, 2=20, 3=20, 4=0, Index=&SpawnGroup(), Num=::NewEnumerator3)
     B=$Emitter.MasterSizeMultiplier
    )
    UseOwnerScale=true
   )
   &AddVelocity(
    Velocity=&MaskVectorBySpawnGroup(Masked Vector=$Particles.ShapeLocation.ShapeNormal, Passthrough Vector=<0 0 0>, Spawn Group Mask=4)
    Velocity Speed Scale=&Multiply_Float(A=&RandomRangeFloat(Minimum=50, Maximum=80), B=$Emitter.MasterSizeMultiplier)
   )
  ]
  particleUpdateScript [
   &ParticleState()
   /SetVariables(Particles.TODColor:LinearColor=&ColorFromCurve(
     ColorCurve=DataInterface() {propertyValues {
       AlphaCurve {keys [<at 0 1 RCIM_Cubic>]}
       BlueCurve {keys [
         <at 0.0043 0.7>
         {arriveTangent=-0.1359, interpMode=RCIM_Cubic, leaveTangent=-0.1359, time=0.1789, value=0.7}
         <at 0.2812 0.6624>
         {arriveTangent=1.2198, interpMode=RCIM_Cubic, leaveTangent=1.2198, time=0.3527, value=1}
         {arriveTangent=-1.0217, interpMode=RCIM_Cubic, leaveTangent=-1.0217, time=0.558, value=1}
         {arriveTangent=-0.9848, interpMode=RCIM_Cubic, leaveTangent=-0.9848, time=0.8029, value=0.54}
         <at 0.8615 0.7011>
         <at 1 0.7>
        ]}
       GreenCurve {keys [
         <at 0.0043 0.4211>
         {arriveTangent=0.9841, interpMode=RCIM_Cubic, leaveTangent=0.9841, time=0.1789, value=0.5785}
         <at 0.2812 0.6936>
         {arriveTangent=1.1069, interpMode=RCIM_Cubic, leaveTangent=1.1069, time=0.3527, value=1}
         {arriveTangent=-0.8536, interpMode=RCIM_Cubic, leaveTangent=-0.8536, time=0.558, value=1}
         {arriveTangent=-1.3917, interpMode=RCIM_Cubic, leaveTangent=-1.3917, time=0.8029, value=0.6157}
         <at 0.8615 0.5776>
         <at 1 0.4138>
        ]}
       RedCurve {keys [
         <at 0.0043 0.3653>
         {arriveTangent=1.9379, interpMode=RCIM_Cubic, leaveTangent=1.9379, time=0.1789, value=0.5542}
         <at 0.2812 0.902>
         {arriveTangent=0.3542, interpMode=RCIM_Cubic, leaveTangent=0.3542, time=0.3527, value=1}
         {arriveTangent=-0.2221, interpMode=RCIM_Cubic, leaveTangent=-0.2221, time=0.558, value=1}
         {arriveTangent=-1.476, interpMode=RCIM_Cubic, leaveTangent=-1.476, time=0.8029, value=0.9}
         <at 0.8615 0.552>
         <at 1 0.3566>
        ]}
      }}
     CurveIndex=Linked() {linkedVariable {name=NPC.FortniteNPC.RemappedTimeOfDay_0-1, type=NiagaraFloat}}
    )) {isSetParameters=true, module=SetVariables_#7}
   /SetVariables(
    Particles.Scale Bright Flare Size:Vector2f=&Multiply_Vector2D_ByFloat(
     Float=$Emitter.MasterSizeMultiplier
     Vector2D=&Multiply_Vector2D(
      A=$Particles.Initial.Bright Flare Size
      B=&Vector2DFromFloat(Value=&FloatFromCurve(FloatCurve=DataInterface() {propertyValues {Curve {keys [
            <at 0 -0.0447>
            <at 0.0141 1.0065>
            <at 0.0287 -0.0508>
            {arriveTangent=-0.3125, interpMode=RCIM_Cubic, leaveTangent=-0.3125, tangentMode=RCTM_User, time=0.0532, value=1}
            {arriveTangent=-3.7966, interpMode=RCIM_Cubic, leaveTangent=-3.7966, tangentMode=RCTM_User, time=1, value=0}
           ]}}}))
     )
    )
    Particles.Update Bright Flare Color:LinearColor=&MakeLinearColorFromVectorAndFloat(
     "Vector (RGB)"=&MakeVectorFromLinearColorRGB(Color=&MultiplyLinearColor(
       Linear Color=$Particles.Initial.Bright Flare color
       Scale Factor=&MakeVector4_FromLinearColor(LinearColor=$Particles.TODColor)
      ))
     "Float (Alpha)"=&FloatFromCurve(FloatCurve=DataInterface() {propertyValues {Curve {keys [
          {arriveTangent=-2.225, interpMode=RCIM_Cubic, leaveTangent=-2.225, tangentMode=RCTM_User, time=0.1, value=1}
          {interpMode=RCIM_Cubic, tangentMode=RCTM_User, time=1, value=0}
         ]}}})
    )
   ) {isSetParameters=true, module=SetVariables_#8}
   /SetVariables(
    Particles.Long Glitch Dynam 1:Vector4f=&MakeVector4(
     X=0
     Y=&RandomRangeFloat(Minimum=-1, Maximum=1)
     Z=&FloatFromCurve(FloatCurve=DataInterface() {propertyValues {Curve {keys [
          <at 0 2 RCIM_Cubic>
          {arriveTangent=-6.2285, interpMode=RCIM_Cubic, leaveTangent=-6.2285, tangentMode=RCTM_User, time=0.5612, value=1.0841}
          {arriveTangent=-0.8021, interpMode=RCIM_Cubic, leaveTangent=-0.8021, tangentMode=RCTM_User, time=1, value=0}
         ]}}})
     W=0
    )
    Particles.Long Glitch Dynam 2:Vector4f=&MakeVector4(
     X=0
     Y=0
     Z=0
     W=&FloatFromCurve(FloatCurve=DataInterface() {propertyValues {Curve {keys [
          {arriveTangent=4.1785, interpMode=RCIM_Cubic, leaveTangent=4.1785, tangentMode=RCTM_User, time=0, value=0}
          {arriveTangent=0.9995, interpMode=RCIM_Cubic, leaveTangent=0.9995, tangentMode=RCTM_User, time=0.2223, value=0.6825}
          {arriveTangent=-0.0405, interpMode=RCIM_Cubic, leaveTangent=-0.0405, tangentMode=RCTM_User, time=1, value=1}
         ]}}})
    )
    Particles.Scale Long Glitch Size:Vector3f=&Multiply_VectorByFloat(
     Vector=&Multiply_Vector(
      A=$Particles.Initial.Long Glitch Size
      B=&Multiply_VectorByFloat(
       Vector=<1 1 1>
       Float=&Multiply_Float(
        A=0.08
        B=&FloatFromCurve(
         FloatCurve=DataInterface() {propertyValues {Curve {keys [
             {arriveTangent=5.4245, interpMode=RCIM_Cubic, leaveTangent=5.4245, tangentMode=RCTM_User, time=0, value=5}
             {arriveTangent=-3.9614, interpMode=RCIM_Cubic, leaveTangent=-3.9614, time=0.0963, value=8.5355}
             {arriveTangent=-22.2662, interpMode=RCIM_Cubic, leaveTangent=-22.2662, time=0.2702, value=3.9295}
             <at 0.3 4>
             <at 0.4 11.5>
             {arriveTangent=0.1532, interpMode=RCIM_Cubic, leaveTangent=0.1532, tangentMode=RCTM_User, time=0.5, value=2.875}
             <at 0.6 3>
             <at 0.7 11.5>
             <at 0.8 2.5>
             <at 0.9 2>
             {arriveTangent=-0.0405, interpMode=RCIM_Cubic, leaveTangent=-0.0405, tangentMode=RCTM_User, time=1, value=0.5}
            ]}}}
         Scale Curve=&Multiply_Float(
          A=0.6
          B=&ClampFloat(
           Float=&FN_LOD_Distance_Remap(Start Distance=200, End Distance=400, Minimum Output=0.5, Maximum Output=1)
           Min=0.5
          )
         )
        )
       )
      )
     )
     Float=$Emitter.MasterSizeMultiplier
    )
    Particles.Update Long Glitch Color:LinearColor=&HueShiftLinearColor(
     Hue=&RandomRangeFloat(Minimum=0.9, Maximum=1)
     Saturation=&RandomRangeFloat(Minimum=3, Maximum=1)
     Value=&RandomRangeFloat(Minimum=0.7, Maximum=1)
     Linear Color=&MultiplyLinearColor(Linear Color=$Emitter.ColorA, Scale Factor=&MakeVector4_FromLinearColor(LinearColor=$Particles.TODColor))
    )
   ) {isSetParameters=true, module=SetVariables_#9}
   /SetVariables(
    Particles.Scale Square Glitch Size:Vector3f=&Multiply_VectorByFloat(
     Vector=&Multiply_Vector(
      A=$Particles.Initial.Square Glitch Size
      B=&Multiply_VectorByFloat(
       Vector=<1 1 1>
       Float=&Multiply_Float(
        A=0.08
        B=&FloatFromCurve(
         FloatCurve=DataInterface() {propertyValues {Curve {keys [
             {arriveTangent=5.4245, interpMode=RCIM_Cubic, leaveTangent=5.4245, tangentMode=RCTM_User, time=0, value=5}
             {arriveTangent=-3.9614, interpMode=RCIM_Cubic, leaveTangent=-3.9614, time=0.0963, value=8.5355}
             {arriveTangent=-22.2662, interpMode=RCIM_Cubic, leaveTangent=-22.2662, time=0.2702, value=3.9295}
             <at 0.3 4>
             <at 0.4 11.5>
             {arriveTangent=0.1532, interpMode=RCIM_Cubic, leaveTangent=0.1532, tangentMode=RCTM_User, time=0.5, value=2.875}
             <at 0.6 3>
             <at 0.7 11.5>
             <at 0.8 2.5>
             <at 0.9 2>
             {arriveTangent=-0.0405, interpMode=RCIM_Cubic, leaveTangent=-0.0405, tangentMode=RCTM_User, time=1, value=0.5}
            ]}}}
         Scale Curve=&Multiply_Float(
          A=0.6
          B=&ClampFloat(
           Float=&FN_LOD_Distance_Remap(Start Distance=200, End Distance=400, Minimum Output=0.5, Maximum Output=1)
           Min=0.5
          )
         )
        )
       )
      )
     )
     Float=$Emitter.MasterSizeMultiplier
    )
    Particles.Square Glitch Dynam 1:Vector4f=&MakeVector4(
     X=0
     Y=&RandomRangeFloat(Minimum=-1, Maximum=1)
     Z=&FloatFromCurve(FloatCurve=DataInterface() {propertyValues {Curve {keys [
          <at 0 2 RCIM_Cubic>
          {arriveTangent=-6.2285, interpMode=RCIM_Cubic, leaveTangent=-6.2285, tangentMode=RCTM_User, time=0.5612, value=1.0841}
          {arriveTangent=-0.8021, interpMode=RCIM_Cubic, leaveTangent=-0.8021, tangentMode=RCTM_User, time=1, value=0}
         ]}}})
     W=0
    )
    Particles.Square Glitch Dynam 2:Vector4f=&MakeVector4(
     X=&RandomRangeFloat(Minimum=-1, Maximum=1)
     Y=0
     Z=0
     W=&FloatFromCurve(FloatCurve=DataInterface() {propertyValues {Curve {keys [
          {arriveTangent=4.1785, interpMode=RCIM_Cubic, leaveTangent=4.1785, tangentMode=RCTM_User, time=0, value=0}
          {arriveTangent=0.9995, interpMode=RCIM_Cubic, leaveTangent=0.9995, tangentMode=RCTM_User, time=0.2223, value=0.6825}
          {arriveTangent=-0.0405, interpMode=RCIM_Cubic, leaveTangent=-0.0405, tangentMode=RCTM_User, time=1, value=1}
         ]}}})
    )
    Particles.Update Square Glitch Color:LinearColor=&HueShiftLinearColor(
     Hue=&RandomRangeFloat(Minimum=0.85, Maximum=1)
     Saturation=1
     Value=1
     Linear Color=&MultiplyLinearColor(Linear Color=$Emitter.ColorB, Scale Factor=&MakeVector4_FromLinearColor(LinearColor=$Particles.TODColor))
    )
   ) {isSetParameters=true, module=SetVariables_#10}
   &Drag(
    Drag=&MaskFloatBySpawnGroup(
     Masked Float=&FloatFromCurve(
      FloatCurve=DataInterface() {propertyValues {Curve {keys [
          {arriveTangent=2.0771, interpMode=RCIM_Cubic, leaveTangent=2.0771, tangentMode=RCTM_User, time=0, value=0}
          {interpMode=RCIM_Cubic, tangentMode=RCTM_User, time=1, value=1}
         ]}}}
      Scale Curve=10
     )
     Passthrough Float=0
     Spawn Group Mask=4
    )
    Rotational Drag:NiagaraFloat=&MaskFloatBySpawnGroup(
     Masked Float=&FloatFromCurve(
      FloatCurve=DataInterface() {propertyValues {Curve {keys [
          {arriveTangent=2.0771, interpMode=RCIM_Cubic, leaveTangent=2.0771, tangentMode=RCTM_User, time=0, value=0}
          {interpMode=RCIM_Cubic, tangentMode=RCTM_User, time=1, value=1}
         ]}}}
      Scale Curve=3
     )
     Passthrough Float=0
     Spawn Group Mask=4
    )
    Use Linear Drag=true
    Use Rotational Drag=true
   )
   &SolveForcesAndVelocity(Clamp Velocity=false, Limit Acceleration=false)
  ]
  rendererClasses [&NiagaraSpriteRendererProperties, &NiagaraMeshRendererProperties]
  renderers [
   {rendererClass=&NiagaraSpriteRendererProperties}
   {rendererClass=&NiagaraMeshRendererProperties, rendererIndex=1}
   {rendererClass=&NiagaraMeshRendererProperties, rendererIndex=2}
  ]
 }
 PropSpawn_Glitch001 {
  emitterSpawnScript [/SetVariables(
    Emitter.ColorA:LinearColor=<rgba 0 0.3281 0.0898 1>
    Emitter.ColorB:LinearColor=<rgba 0 0.3177 0.0115 1>
    Emitter.MasterSizeMultiplier=0.7
    Emitter.ShapeLocationScale:Vector3f=<1.1 1.5 0.65>
   ) {isSetParameters=true, module=SetVariables_#1}]
  emitterUpdateScript [
   &EmitterState(Life Cycle Mode=::NewEnumerator1, Loop Behavior=::NewEnumerator1, Loop Duration=1, UseLoopDelay=false)
   &SpawnBurst_Instantaneous(Use Spawn Probability=false, UseLoopCountLimit=false)
   &SpawnBurst_Instantaneous(Spawn Count=10, Spawn Group=2, Use Spawn Probability=false, UseLoopCountLimit=false) {module=SpawnBurst_Instantaneous008}
   &SpawnBurst_Instantaneous(Spawn Count=5, Spawn Group=2, Spawn Time=0.15, Use Spawn Probability=false, UseLoopCountLimit=false) {module=SpawnBurst_Instantaneous012}
   &SpawnBurst_Instantaneous(Spawn Count=3, Spawn Group=3, Use Spawn Probability=false, UseLoopCountLimit=false) {module=SpawnBurst_Instantaneous011}
   &SpawnBurst_Instantaneous(Spawn Count=3, Spawn Group=3, Spawn Time=0.05, Use Spawn Probability=false, UseLoopCountLimit=false) {module=SpawnBurst_Instantaneous010}
   &SpawnBurst_Instantaneous(Spawn Count=5, Spawn Group=3, Spawn Time=0.1, Use Spawn Probability=false, UseLoopCountLimit=false) {module=SpawnBurst_Instantaneous009}
  ]
  particleSpawnScript [
   /SetVariables(Particles.VisibilityTag=&SpawnGroup()) {isSetParameters=true, module=SetVariables_#2}
   /SetVariables(Particles.Random Range:NiagaraFloat=&RandomRangeFloat(Minimum=0.2, Maximum=0.8)) {isSetParameters=true, module=SetVariables_#3}
   &InitializeParticle(
    Color Mode=::NewEnumerator0
    Lifetime=&MergeHelper_FloatSwitch(
     0=0.27
     1=0.3
     2=&RandomRangeFloat(Minimum=0.25, Maximum=0.3)
     3=&RandomRangeFloat(Minimum=0.25, Maximum=0.3)
     4=0.2
     6=&RandomRangeFloat(Minimum=0.55, Maximum=0.55)
     7=&RandomRangeFloat(Minimum=0.5, Maximum=0.62)
     8=&RandomRangeFloat(Minimum=0.25, Maximum=0.5)
     9=&RandomRangeFloat(Minimum=0.25, Maximum=0.5)
     Index=&SpawnGroup()
     Num=::NewEnumerator3
    )
    Position Offset=<20 0 0>
    UsePositionOffset=true
    Write Material Random=true
   ) {module=InitializeParticle001}
   &CameraOffset(Camera Offset Amount=&MergeHelper_FloatSwitch(0=2, 1=2, 2=10, 3=10, 4=2, Index=0, Num=::NewEnumerator3))
   /SetVariables(
    Particles.Bright Flare Size:Vector2f=&Vector2DFromFloat(Value=130)
    Particles.Bright Flare color:LinearColor=$Emitter.ColorB
   ) {isSetParameters=true, module=SetVariables_#4}
   /SetVariables(
    Particles.Long Glitch Dynam 1:Vector4f=&MakeVector4(X=0, Y=0, Z=0, W=&RandomRangeFloat(Minimum=-1, Maximum=0.1))
    Particles.Long Glitch Dynam 2:Vector4f=&MakeVector4(X=&RandomRangeFloat(Minimum=-100, Maximum=100), Y=0, Z=0, W=0)
    Particles.Long Glitch Size:Vector3f=&RandomRangeVector(Minimum=<1 3 0.5>, Maximum=<2 5 1>)
   ) {isSetParameters=true, module=SetVariables_#5}
   /SetVariables(
    Particles.Square Glitch Dynam 1:Vector4f=&MakeVector4(X=0, Y=0, Z=0, W=&RandomRangeFloat(Minimum=-1, Maximum=0.1))
    Particles.Square Glitch Dynam 2:Vector4f=&MakeVector4(X=&RandomRangeFloat(Minimum=-100, Maximum=100), Y=0, Z=0, W=0)
    Particles.Square Glitch Size:Vector3f=&RandomRangeVector(Minimum=<1 1 1>, Maximum=<1.5 1.5 1.5>)
   ) {isSetParameters=true, module=SetVariables_#6}
   &ShapeLocation(
    MaskSpawnGroup=false
    Non Uniform Scale=$Emitter.ShapeLocationScale
    Override Seed=false
    Sphere Radius=&Multiply_Float(
     A=&MergeHelper_FloatSwitch(0=0, 1=0, 2=20, 3=20, 4=0, Index=&SpawnGroup(), Num=::NewEnumerator3)
     B=$Emitter.MasterSizeMultiplier
    )
    UseOwnerScale=true
   )
   &AddVelocity(
    Velocity=&MaskVectorBySpawnGroup(Masked Vector=$Particles.ShapeLocation.ShapeNormal, Passthrough Vector=<0 0 0>, Spawn Group Mask=4)
    Velocity Speed Scale=&Multiply_Float(A=&RandomRangeFloat(Minimum=50, Maximum=80), B=$Emitter.MasterSizeMultiplier)
   )
  ]
  particleUpdateScript [
   &ParticleState()
   /SetVariables(Particles.TODColor:LinearColor=&ColorFromCurve(
     ColorCurve=DataInterface() {propertyValues {
       AlphaCurve {keys [<at 0 1 RCIM_Cubic>]}
       BlueCurve {keys [
         <at 0.0043 0.7>
         {arriveTangent=-0.1359, interpMode=RCIM_Cubic, leaveTangent=-0.1359, time=0.1789, value=0.7}
         <at 0.2812 0.6624>
         {arriveTangent=1.2198, interpMode=RCIM_Cubic, leaveTangent=1.2198, time=0.3527, value=1}
         {arriveTangent=-1.0217, interpMode=RCIM_Cubic, leaveTangent=-1.0217, time=0.558, value=1}
         {arriveTangent=-0.9848, interpMode=RCIM_Cubic, leaveTangent=-0.9848, time=0.8029, value=0.54}
         <at 0.8615 0.7011>
         <at 1 0.7>
        ]}
       GreenCurve {keys [
         <at 0.0043 0.4211>
         {arriveTangent=0.9841, interpMode=RCIM_Cubic, leaveTangent=0.9841, time=0.1789, value=0.5785}
         <at 0.2812 0.6936>
         {arriveTangent=1.1069, interpMode=RCIM_Cubic, leaveTangent=1.1069, time=0.3527, value=1}
         {arriveTangent=-0.8536, interpMode=RCIM_Cubic, leaveTangent=-0.8536, time=0.558, value=1}
         {arriveTangent=-1.3917, interpMode=RCIM_Cubic, leaveTangent=-1.3917, time=0.8029, value=0.6157}
         <at 0.8615 0.5776>
         <at 1 0.4138>
        ]}
       RedCurve {keys [
         <at 0.0043 0.3653>
         {arriveTangent=1.9379, interpMode=RCIM_Cubic, leaveTangent=1.9379, time=0.1789, value=0.5542}
         <at 0.2812 0.902>
         {arriveTangent=0.3542, interpMode=RCIM_Cubic, leaveTangent=0.3542, time=0.3527, value=1}
         {arriveTangent=-0.2221, interpMode=RCIM_Cubic, leaveTangent=-0.2221, time=0.558, value=1}
         {arriveTangent=-1.476, interpMode=RCIM_Cubic, leaveTangent=-1.476, time=0.8029, value=0.9}
         <at 0.8615 0.552>
         <at 1 0.3566>
        ]}
      }}
     CurveIndex=Linked() {linkedVariable {name=NPC.FortniteNPC.RemappedTimeOfDay_0-1, type=NiagaraFloat}}
    )) {isSetParameters=true, module=SetVariables_#7}
   /SetVariables(
    Particles.Scale Bright Flare Size:Vector2f=&Multiply_Vector2D_ByFloat(
     Float=$Emitter.MasterSizeMultiplier
     Vector2D=&Multiply_Vector2D(
      A=$Particles.Initial.Bright Flare Size
      B=&Vector2DFromFloat(Value=&FloatFromCurve(FloatCurve=DataInterface() {propertyValues {Curve {keys [
            <at 0 -0.0447>
            <at 0.0141 1.0065>
            <at 0.0287 -0.0508>
            {arriveTangent=-0.3125, interpMode=RCIM_Cubic, leaveTangent=-0.3125, tangentMode=RCTM_User, time=0.0532, value=1}
            {arriveTangent=-3.7966, interpMode=RCIM_Cubic, leaveTangent=-3.7966, tangentMode=RCTM_User, time=1, value=0}
           ]}}}))
     )
    )
    Particles.Update Bright Flare Color:LinearColor=&MakeLinearColorFromVectorAndFloat(
     "Vector (RGB)"=&MakeVectorFromLinearColorRGB(Color=&MultiplyLinearColor(
       Linear Color=$Particles.Initial.Bright Flare color
       Scale Factor=&MakeVector4_FromLinearColor(LinearColor=$Particles.TODColor)
      ))
     "Float (Alpha)"=&FloatFromCurve(FloatCurve=DataInterface() {propertyValues {Curve {keys [
          {arriveTangent=-2.225, interpMode=RCIM_Cubic, leaveTangent=-2.225, tangentMode=RCTM_User, time=0.1, value=1}
          {interpMode=RCIM_Cubic, tangentMode=RCTM_User, time=1, value=0}
         ]}}})
    )
   ) {isSetParameters=true, module=SetVariables_#8}
   /SetVariables(
    Particles.Long Glitch Dynam 1:Vector4f=&MakeVector4(
     X=0
     Y=&RandomRangeFloat(Minimum=-1, Maximum=1)
     Z=&FloatFromCurve(FloatCurve=DataInterface() {propertyValues {Curve {keys [
          <at 0 2 RCIM_Cubic>
          {arriveTangent=-6.2285, interpMode=RCIM_Cubic, leaveTangent=-6.2285, tangentMode=RCTM_User, time=0.5612, value=1.0841}
          {arriveTangent=-0.8021, interpMode=RCIM_Cubic, leaveTangent=-0.8021, tangentMode=RCTM_User, time=1, value=0}
         ]}}})
     W=0
    )
    Particles.Long Glitch Dynam 2:Vector4f=&MakeVector4(
     X=0
     Y=0
     Z=0
     W=&FloatFromCurve(FloatCurve=DataInterface() {propertyValues {Curve {keys [
          {arriveTangent=4.1785, interpMode=RCIM_Cubic, leaveTangent=4.1785, tangentMode=RCTM_User, time=0, value=0}
          {arriveTangent=0.9995, interpMode=RCIM_Cubic, leaveTangent=0.9995, tangentMode=RCTM_User, time=0.2223, value=0.6825}
          {arriveTangent=-0.0405, interpMode=RCIM_Cubic, leaveTangent=-0.0405, tangentMode=RCTM_User, time=1, value=1}
         ]}}})
    )
    Particles.Scale Long Glitch Size:Vector3f=&Multiply_VectorByFloat(
     Vector=&Multiply_Vector(
      A=$Particles.Initial.Long Glitch Size
      B=&Multiply_VectorByFloat(
       Vector=<1 1 1>
       Float=&Multiply_Float(
        A=0.08
        B=&FloatFromCurve(
         FloatCurve=DataInterface() {propertyValues {Curve {keys [
             {arriveTangent=5.4245, interpMode=RCIM_Cubic, leaveTangent=5.4245, tangentMode=RCTM_User, time=0, value=5}
             {arriveTangent=-3.9614, interpMode=RCIM_Cubic, leaveTangent=-3.9614, time=0.0963, value=8.5355}
             {arriveTangent=-22.2662, interpMode=RCIM_Cubic, leaveTangent=-22.2662, time=0.2702, value=3.9295}
             <at 0.3 4>
             <at 0.4 11.5>
             {arriveTangent=0.1532, interpMode=RCIM_Cubic, leaveTangent=0.1532, tangentMode=RCTM_User, time=0.5, value=2.875}
             <at 0.6 3>
             <at 0.7 11.5>
             <at 0.8 2.5>
             <at 0.9 2>
             {arriveTangent=-0.0405, interpMode=RCIM_Cubic, leaveTangent=-0.0405, tangentMode=RCTM_User, time=1, value=0.5}
            ]}}}
         Scale Curve=&Multiply_Float(
          A=0.6
          B=&ClampFloat(
           Float=&FN_LOD_Distance_Remap(Start Distance=200, End Distance=400, Minimum Output=0.5, Maximum Output=1)
           Min=0.5
          )
         )
        )
       )
      )
     )
     Float=$Emitter.MasterSizeMultiplier
    )
    Particles.Update Long Glitch Color:LinearColor=&HueShiftLinearColor(
     Hue=&RandomRangeFloat(Minimum=0.85, Maximum=1)
     Saturation=&RandomRangeFloat(Minimum=3, Maximum=1)
     Value=&RandomRangeFloat(Minimum=0.7, Maximum=1)
     Linear Color=&MultiplyLinearColor(Linear Color=$Emitter.ColorA, Scale Factor=&MakeVector4_FromLinearColor(LinearColor=$Particles.TODColor))
    )
   ) {isSetParameters=true, module=SetVariables_#9}
   /SetVariables(
    Particles.Scale Square Glitch Size:Vector3f=&Multiply_VectorByFloat(
     Vector=&Multiply_Vector(
      A=$Particles.Initial.Square Glitch Size
      B=&Multiply_VectorByFloat(
       Vector=<1 1 1>
       Float=&Multiply_Float(
        A=0.08
        B=&FloatFromCurve(
         FloatCurve=DataInterface() {propertyValues {Curve {keys [
             {arriveTangent=5.4245, interpMode=RCIM_Cubic, leaveTangent=5.4245, tangentMode=RCTM_User, time=0, value=5}
             {arriveTangent=-3.9614, interpMode=RCIM_Cubic, leaveTangent=-3.9614, time=0.0963, value=8.5355}
             {arriveTangent=-22.2662, interpMode=RCIM_Cubic, leaveTangent=-22.2662, time=0.2702, value=3.9295}
             <at 0.3 4>
             <at 0.4 11.5>
             {arriveTangent=0.1532, interpMode=RCIM_Cubic, leaveTangent=0.1532, tangentMode=RCTM_User, time=0.5, value=2.875}
             <at 0.6 3>
             <at 0.7 11.5>
             <at 0.8 2.5>
             <at 0.9 2>
             {arriveTangent=-0.0405, interpMode=RCIM_Cubic, leaveTangent=-0.0405, tangentMode=RCTM_User, time=1, value=0.5}
            ]}}}
         Scale Curve=&Multiply_Float(
          A=0.6
          B=&ClampFloat(
           Float=&FN_LOD_Distance_Remap(Start Distance=200, End Distance=400, Minimum Output=0.5, Maximum Output=1)
           Min=0.5
          )
         )
        )
       )
      )
     )
     Float=$Emitter.MasterSizeMultiplier
    )
    Particles.Square Glitch Dynam 1:Vector4f=&MakeVector4(
     X=0
     Y=&RandomRangeFloat(Minimum=-1, Maximum=1)
     Z=&FloatFromCurve(FloatCurve=DataInterface() {propertyValues {Curve {keys [
          <at 0 2 RCIM_Cubic>
          {arriveTangent=-6.2285, interpMode=RCIM_Cubic, leaveTangent=-6.2285, tangentMode=RCTM_User, time=0.5612, value=1.0841}
          {arriveTangent=-0.8021, interpMode=RCIM_Cubic, leaveTangent=-0.8021, tangentMode=RCTM_User, time=1, value=0}
         ]}}})
     W=0
    )
    Particles.Square Glitch Dynam 2:Vector4f=&MakeVector4(
     X=&RandomRangeFloat(Minimum=-1, Maximum=1)
     Y=0
     Z=0
     W=&FloatFromCurve(FloatCurve=DataInterface() {propertyValues {Curve {keys [
          {arriveTangent=4.1785, interpMode=RCIM_Cubic, leaveTangent=4.1785, tangentMode=RCTM_User, time=0, value=0}
          {arriveTangent=0.9995, interpMode=RCIM_Cubic, leaveTangent=0.9995, tangentMode=RCTM_User, time=0.2223, value=0.6825}
          {arriveTangent=-0.0405, interpMode=RCIM_Cubic, leaveTangent=-0.0405, tangentMode=RCTM_User, time=1, value=1}
         ]}}})
    )
    Particles.Update Square Glitch Color:LinearColor=&HueShiftLinearColor(
     Hue=&RandomRangeFloat(Minimum=0.85, Maximum=1)
     Saturation=1
     Value=1
     Linear Color=&MultiplyLinearColor(Linear Color=$Emitter.ColorB, Scale Factor=&MakeVector4_FromLinearColor(LinearColor=$Particles.TODColor))
    )
   ) {isSetParameters=true, module=SetVariables_#10}
   &Drag(
    Drag=&MaskFloatBySpawnGroup(
     Masked Float=&FloatFromCurve(
      FloatCurve=DataInterface() {propertyValues {Curve {keys [
          {arriveTangent=2.0771, interpMode=RCIM_Cubic, leaveTangent=2.0771, tangentMode=RCTM_User, time=0, value=0}
          {interpMode=RCIM_Cubic, tangentMode=RCTM_User, time=1, value=1}
         ]}}}
      Scale Curve=10
     )
     Passthrough Float=0
     Spawn Group Mask=4
    )
    Rotational Drag:NiagaraFloat=&MaskFloatBySpawnGroup(
     Masked Float=&FloatFromCurve(
      FloatCurve=DataInterface() {propertyValues {Curve {keys [
          {arriveTangent=2.0771, interpMode=RCIM_Cubic, leaveTangent=2.0771, tangentMode=RCTM_User, time=0, value=0}
          {interpMode=RCIM_Cubic, tangentMode=RCTM_User, time=1, value=1}
         ]}}}
      Scale Curve=3
     )
     Passthrough Float=0
     Spawn Group Mask=4
    )
    Use Linear Drag=true
    Use Rotational Drag=true
   )
   &SolveForcesAndVelocity(Clamp Velocity=false, Limit Acceleration=false)
  ]
  rendererClasses [&NiagaraSpriteRendererProperties, &NiagaraMeshRendererProperties]
  renderers [
   {rendererClass=&NiagaraSpriteRendererProperties}
   {rendererClass=&NiagaraMeshRendererProperties, rendererIndex=1}
   {rendererClass=&NiagaraMeshRendererProperties, rendererIndex=2}
  ]
 }
}
summary {
 systemName=NS_Emote_CopperMoggy_PropSpawn
 emitters [
  {
   emitterName=PropSpawn_Glitch
   rendererClasses [&NiagaraSpriteRendererProperties, &NiagaraMeshRendererProperties]
  }
  {
   emitterName=PropSpawn_Glitch001
   rendererClasses [&NiagaraSpriteRendererProperties, &NiagaraMeshRendererProperties]
  }
 ]
 userVariables [
  User.bIsFrontEnd {type=NiagaraBool}
  User.bIsFrontEndPreview {type=NiagaraBool}
 ]
}
